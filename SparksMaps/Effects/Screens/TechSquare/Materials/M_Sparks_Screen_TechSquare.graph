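# Unlit Sparks stage screen material showing animated tech squares that react to audio beats.
# Panning square textures are thresholded into fills and outlines, combined with noise highlights, gradient fade, ambient glow and a switchable pattern-driven pulse, then passed through MF_Sparks_Screen_Details.
# Static switches select stage color A, B or C from MF_MojoStageColors, with squared copies as secondaries.
# BeatLifetime from the audio parameter collection drives a curve-atlas reactivity value and eased beat envelopes.
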Material /SparksMaps/Effects/Screens/TechSquare/Materials/M_Sparks_Screen_TechSquare
graph {
 properties {
  bAutomaticallySetUsageInEditor=false
  bHasPixelAnimation=true
  bUsedWithNanite=true
  pixelDepthOffsetMode=PDOM_Legacy
  shadingModel=MSM_Unlit
  usedShadingModels=Unlit
 }
 nodes [
  n1 = NamedRerouteUsage()
  n2 = Multiply(A=@n1) {constB=0.5}
  n3 = MaterialFunctionCall!(
   Width=Constant() {r=0.5}
   Sharpness=Constant() {r=0.25}
   Offset=Frac(None=Time())
   Direction Switch
   Compare
   Input Coordinates=NamedRerouteUsage()
  ) {materialFunction=/Engine/Functions/Engine_MaterialFunctions03/Procedurals/GeneratedBand}
  Add(A=@n2, B=Multiply(A=OneMinus(None=@n3.Result), B=Constant() {r=0.1}) {constB=0.1})
  Frac(None=Time())
  n4 = NamedRerouteUsage()
  n5 = Rotator(Coordinate=TextureCoordinate(), Time=ScalarParameter() {defaultValue=2.5, parameterName=AmbientGlowRotation})
  n6 = Multiply(
   A=Power(
    Base=MaterialFunctionCall(UV Channel=@n5) {materialFunction=&LinearGradient}.VGradient
    Exp=ScalarParameter() {defaultValue=3, parameterName=AmbientGlowPow}
   ) {constExponent=3}
   B=ScalarParameter() {defaultValue=0.1, parameterName=AmbientGlowMult}
  ) {constB=0.05}
  n7 = Time()
  n8 = ScalarParameter() {defaultValue=0.25, parameterName=Param_2}
  n9 = TextureCoordinate()
  n10 = ScalarParameter() {defaultValue=1, parameterName=Square Pulse Tile}
  n11 = TextureSample(UVs=Panner(Coordinate=Multiply(A=@n9, B=@n10)) {speedX=0.05, speedY=0.05}) {samplerType=SAMPLERTYPE_LinearColor, texture=&/T_Sparks_Screen_TechSquare_Pattern}
  n12 = Frac(None=Add(A=Multiply(A=@n7, B=@n8) {constB=0.5}, B=@n11.R))
  n13 = ScalarParameter() {defaultValue=0.9}
  n14 = ScalarParameter() {defaultValue=1, parameterName=Param_1}
  n15 = Constant() {r=15}
  n16 = Multiply(A=Add(A=Subtract(A=@n12), B=@n11.G), B=Clamp(None=Multiply(A=@n11.R, B=@n15)) {maxDefault=10, minDefault=3}) {constB=6}
  n17 = Constant() {r=20}
  n18 = Sine(None=MaterialFunctionCall(Multiplier=Clamp(None=@n16, Max=@n17) {maxDefault=8}) {materialFunction=&Pi}.Result) {period=6.2832}
  n19 = Multiply(
   A=Multiply(A=OneMinus(None=@n12), B=SmoothStep(Min=@n13, Max=@n14, Value=@n18) {constMax=0.8, constMin=0.7})
   B=Step(Y=@n11.R, X=Constant() {r=1}) {constX=0.5}
  )
  n20 = TextureCoordinate()
  n21 = ScalarParameter() {defaultValue=1, parameterName=Square Highlights Tile}
  n22 = Time()
  n23 = Constant() {r=2}
  n24 = TextureSample(UVs=Panner(Coordinate=Multiply(A=@n20, B=@n21), Time=Multiply(A=@n22, B=@n23)) {speedX=0.1, speedY=0.1}) {texture=/Game/Global/Textures/Default/Noise/T_SoftCells_Negative}
  n25 = Multiply(A=NamedRerouteUsage(), B=NamedRerouteUsage())
  n26 = Multiply(A=Multiply(A=Power(Base=@n24.G) {constExponent=5}, B=Saturate(None=@n25)), B=Constant() {r=5})
  n27 = Add(A=NamedRerouteUsage()) {constB=0.5}
  n28 = Multiply(A=@n2, B=NamedRerouteUsage())
  n29 = Constant() {r=0.0075}
  n30 = Add(A=@n29) {constB=0.01}
  n31 = TextureCoordinate() {uTiling=0.75, vTiling=0.75}
  n32 = ScalarParameter() {defaultValue=1, parameterName=Square Tile}
  n33 = TextureSample(UVs=Panner(Coordinate=Multiply(A=@n31, B=@n32)) {speedX=0.025, speedY=0.05}) {texture=&/T_Sparks_Screen_TechSquares}
  n34 = Multiply(A=Saturate(None=SmoothStep(Min=@n29, Max=@n30, Value=@n33.G) {constMax=0.25, constMin=0.2})) {constB=0.35}
  n35 = LinearInterpolate(Alpha=Power(Base=MaterialFunctionCall() {materialFunction=&LinearGradient}.VGradient) {constExponent=5}) {constA=1, constB=0}
  n36 = Multiply(A=OneMinus(None=@n1), B=Multiply(A=Multiply(A=@n34, B=@n35), B=NamedRerouteUsage()))
  n37 = MaterialFunctionCall() {materialFunction=&LinearGradient}
  n38 = ScalarParameter() {defaultValue=3, parameterName=SquareFadeGradientPower}
  n39 = Multiply(
   A=Add(A=Multiply(A=@n26, B=@n27), B=Add(A=@n28, B=@n36))
   B=LinearInterpolate(
    A=ScalarParameter() {defaultValue=0.05, parameterName=SquareFadeUpper}
    B=ScalarParameter() {defaultValue=1, parameterName=SquareFadeLower}
    Alpha=Power(Base=@n37.VGradient, Exp=@n38) {constExponent=5}
   ) {constA=1, constB=0.25}
  )
  n40 = Add(A=@n39, B=Multiply(A=Saturate(None=@n19), B=NamedRerouteUsage()))
  MaterialFunctionCall!(Emissive=Add(A=Multiply(A=@n4, B=@n6), B=StaticSwitchParameter(True=@n40, False=@n39) {parameterName=UseSquarePulse}), Scanlines Tile) {materialFunction=/SparksMaps/Effects/Screens/Materials/MF/MF_Sparks_Screen_Details}
  n41 = NamedRerouteUsage()
  n42 = NamedRerouteUsage()
  NamedRerouteDeclaration(None=StaticSwitchParameter(True=NamedRerouteUsage(), False=StaticSwitchParameter(True=@n41, False=@n42) {parameterName=IsColorA}) {parameterName=IsColorC}) {name=PrimaryColor, nodeColor=<rgba 0 0 0 1>}
  n43 = CurveAtlasRowParameter(CurveTime=CollectionParameter() {collection=/FMAudioReactiveData/AudioAnalysis/MPC_Sparks_Audio_Parameters, parameterName=BeatLifetime}) {
   atlas=/SparksMaps/Effects/Screens/Curves/CurveAtlas_Audio_Beat_A
   curve=/SparksMaps/Effects/Screens/Curves/Curve_Audio_Beat_Remap_001
   parameterName=ReactivityCurve
  }
  NamedRerouteDeclaration(None=MaterialFunctionCall(Input=@n43, Input Low=Constant(), Input High=Constant() {r=1}, Target Low=Constant() {r=0.75}, Target High=Constant() {r=2}) {materialFunction=&RemapValueRange}.Result) {name=Reactivity, nodeColor=<rgba 0 0 0 1>}
  n44 = MaterialFunctionCall!(CustomUPosition) {materialFunction=/SparksStageShow/Systems/StageManager/FXManager/MaterialFunctions/MF_MojoStageColors}
  NamedRerouteDeclaration(None=@n44.Primary) {name=Color A Primary, nodeColor=<rgba 0 1 0.0859 1>}
  NamedRerouteDeclaration(None=Multiply(A=@n44.Primary, B=@n44.Primary)) {name=Color A Secondary, nodeColor=<rgba 1 0 0.2344 1>}
  NamedRerouteDeclaration(None=@n44.Secondary) {name=Color B Primary, nodeColor=<rgba 0 0.5547 1 1>}
  NamedRerouteDeclaration(None=Multiply(A=@n44.Secondary, B=@n44.Secondary)) {name=Color B Secondary, nodeColor=<rgba 0.875 1 0 1>}
  NamedRerouteDeclaration(None=@n44.Tertiary) {name=Color C Primary, nodeColor=<rgba 0 0.6484 1 1>}
  NamedRerouteDeclaration(None=Multiply(A=@n44.Tertiary, B=@n44.Tertiary)) {name=Color C Secondary, nodeColor=<rgba 0.9688 1 0 1>}
  n45 = NamedRerouteUsage()
  n46 = NamedRerouteUsage()
  NamedRerouteDeclaration(None=StaticSwitchParameter(True=NamedRerouteUsage(), False=StaticSwitchParameter(True=@n45, False=@n46) {parameterName=IsColorA}) {parameterName=IsColorC}) {name=SecondaryColor, nodeColor=<rgba 0 0 0 1>}
  n47 = Multiply(A=TextureCoordinate(), B=ScalarParameter() {defaultValue=1, parameterName=Square Tile})
  NamedRerouteDeclaration(None=TextureSample(UVs=Panner(Coordinate=@n47) {speedX=0.025, speedY=0.025}) {texture=&T_Sparks_Screen_TechSquares}.G) {name=SquareTexture, nodeColor=<rgba 0 0 0 1>}
  n48 = LinearInterpolate(Alpha=NamedRerouteUsage()) {constA=0.04, constB=0.01}
  n49 = Add(A=@n48, B=Constant() {r=0.01}) {constB=0.025}
  n50 = NamedRerouteUsage()
  NamedRerouteDeclaration(None=Saturate(None=SmoothStep(Min=@n48, Max=@n49, Value=@n50) {constMax=0.25, constMin=0.2})) {name=MainSquares, nodeColor=<rgba 0 0 0 1>}
  n51 = Add(A=@n49, B=Constant() {r=0.01}) {constB=0.03}
  NamedRerouteDeclaration(None=Saturate(None=SmoothStep(Min=@n51, Max=@n49, Value=@n50) {constMax=0.25, constMin=0.2})) {name=MainSquaresOutlineMult, nodeColor=<rgba 0 0 0 1>}
  n52 = Constant() {desc=Anim In Length, r=0.1}
  n53 = CollectionParameter() {collection=&MPC_Sparks_Audio_Parameters, parameterName=BeatLifetime}
  NamedRerouteDeclaration(None=Saturate(None=Divide(A=@n53, B=@n52))) {name=BeatLinearIn, nodeColor=<rgba 0 0 0 1>}
  n54 = Constant() {desc=Anim Out Length, r=0.7}
  n55 = Divide(A=Subtract(A=@n53, B=Constant()), B=@n54)
  NamedRerouteDeclaration(None=OneMinus(None=Saturate(None=@n55))) {name=BeatLinearOut, nodeColor=<rgba 0 0 0 1>}
  n56 = Divide(A=Subtract(A=@n53, B=@n52), B=Add(A=@n52, B=@n54)) {constB=0.5}
  NamedRerouteDeclaration(None=OneMinus(None=Saturate(None=@n56))) {name=BeatLinearOutDelayed, nodeColor=<rgba 0 0 0 1>}
  n57 = OneMinus(None=NamedRerouteUsage())
  n58 = OneMinus(None=Multiply(A=Multiply(A=@n57, B=@n57), B=@n57))
  NamedRerouteDeclaration(None=Multiply(A=Saturate(None=@n58) {desc=Ease Out of Anim}, B=NamedRerouteUsage() {desc=Linear Anim Out})) {name=BeatAnim, nodeColor=<rgba 0 0 0 1>}
  ScalarParameter() {defaultValue=0.01, parameterName=SquareSize}
 ]
}
